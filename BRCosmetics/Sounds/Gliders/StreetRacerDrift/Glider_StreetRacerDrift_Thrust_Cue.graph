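# Sound cue for the thrust sound of the StreetRacerDrift glider cosmetic.
# A Random node picks one of two equally weighted thrust waves and feeds a Modulator at the root for slight per-play variation.
# Playback uses the shared glider thrust attenuation and concurrency assets, the glider loop audio properties sheet and the glider land sound class, at reduced volume.
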SoundCue /BRCosmetics/Sounds/Gliders/StreetRacerDrift/Glider_StreetRacerDrift_Thrust_Cue
root=Modulator_1
nodes [
 {id=Modulator_1, children [Random_1]}
 {id=Random_1, children [WavePlayer_2,WavePlayer_5], properties {weights [1,1]}}
 {id=WavePlayer_2, properties {soundWaveAssetPtr=&/Athena_Glider_StreetRacerDrift_Thrust_02}}
 {id=WavePlayer_5, properties {soundWaveAssetPtr=&/Athena_Glider_StreetRacerDrift_Thrust_01}}
]
properties {
 attenuationSettings=&Fort_Player_Glider_Thrust_Athena_Att
 audioPropertiesSheet=&APS_Cosmetic_Glider_Loop
 maxDistance=12500
 soundClassObject=&Fort_Gliders_Land
 volumeMultiplier=0.25
 concurrencySet [&Athena_Glider_Thrust_Conc]
}
